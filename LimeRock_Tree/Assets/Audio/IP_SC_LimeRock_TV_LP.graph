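# Sound cue for a LimeRock TV prop, playing an attenuated ambient loop with intermittent static bursts.
# A Mixer layers the looping TV wave with a slightly boosted Looping branch that repeatedly plays a random static sample after a random delay, with slight pitch variation.
# A quality-level node trims the static pool from eight variations on its first two levels to four on the last two.
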SoundCue /LimeRock_Tree/Assets/Audio/IP_SC_LimeRock_TV_LP
root=Mixer_0
nodes [
 {id=Delay_0, children [Modulator_0], properties {delayMax=3, delayMin=1.5}}
 {id=Looping_0, children [Delay_0]}
 {id=Mixer_0, children [WavePlayer_0,Looping_0], properties {inputVolume [1,1.2]}}
 {id=Modulator_0, children [QualityLevel_1], properties {pitchMax=1.1, pitchMin=0.9, volumeMax=1, volumeMin=1}}
 {id=QualityLevel_1, children [Random_0,Random_0,Random_4,Random_4]}
 {
  id=Random_0
  children [WavePlayer_1,WavePlayer_2,WavePlayer_3,WavePlayer_4,WavePlayer_5,WavePlayer_6,WavePlayer_7,WavePlayer_8]
  properties {weights [1,1,1,1,1,1,1,1]}
 }
 {
  id=Random_4
  children [WavePlayer_1,WavePlayer_2,WavePlayer_3,WavePlayer_4]
  properties {bShouldExcludeFromBranchCulling=true, weights [1,1,1,1]}
 }
 {id=WavePlayer_0, properties {bLooping=true, soundWaveAssetPtr=&/IP_LimeRock_TV_LP}}
 {id=WavePlayer_1, properties {soundWaveAssetPtr=&/IP_LimeRock_TV_Static_01}}
 {id=WavePlayer_2, properties {soundWaveAssetPtr=&/IP_LimeRock_TV_Static_02}}
 {id=WavePlayer_3, properties {soundWaveAssetPtr=&/IP_LimeRock_TV_Static_03}}
 {id=WavePlayer_4, properties {soundWaveAssetPtr=&/IP_LimeRock_TV_Static_04}}
 {id=WavePlayer_5, properties {soundWaveAssetPtr=&/IP_LimeRock_TV_Static_05}}
 {id=WavePlayer_6, properties {soundWaveAssetPtr=&/IP_LimeRock_TV_Static_06}}
 {id=WavePlayer_7, properties {soundWaveAssetPtr=&/IP_LimeRock_TV_Static_07}}
 {id=WavePlayer_8, properties {soundWaveAssetPtr=&/IP_LimeRock_TV_Static_08}}
]
properties {
 attenuationSettings=&/ATT_LimeRock_Prop_TV
 maxDistance=2100
 soundClassObject=&Fort_Ambient_Props_Athena
 volumeMultiplier=3
}
